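# Physics asset for the Sir Wolf back bling (backpack cosmetic) mesh, defining simple collision and a joint for its dangling ring bone.
# It holds two small capsule bodies, one on the pack bone and one, offset and rotated, on dyn_ring.
# A single constraint links dyn_ring to pack with twist and one swing axis nearly locked and the other given a wide limit, so the ring moves like a hinge.
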PhysicsAsset /BRCosmetics/Accessories/FORT_Backpacks/Backpack_M_MED_SirWolf/Meshes/M_MED_SirWolf_Pack_Physics
bodies {
 dyn_ring [{shapeType=Capsule, length=0.5, radius=0.5, shapeName=dyn_ring_capsule, center=<1 0 0>, rotation=<rot 90 0 0>}]
 pack [{shapeType=Capsule, length=2, radius=1, shapeName=pack_capsule}]
}
constraints [{bone1Name=dyn_ring, bone2Name=pack, swing1LimitDegrees=1, swing2LimitDegrees=80, twistLimitDegrees=1}]
